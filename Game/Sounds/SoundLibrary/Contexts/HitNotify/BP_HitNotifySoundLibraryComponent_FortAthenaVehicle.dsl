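; Sound library component for Athena vehicles that keeps the vehicle's hit-notify sounds matched to the player's new-hitmarkers audio setting.
; On BeginPlay and whenever that setting changes, it picks the Modern or Legacy vehicle profile tag.
; SelectHitmarkerTypeByTag then uses the SoundLibrary subsystem to remove both profiles' libraries from a valid owning actor and add the chosen one, so the two never stack.
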(event SoundLibrary|EventOnSetNewHitmarkersEnabledChanged
  (Default|SelectHitmarkerTypebyTag (select (FortAudioOptions|GetNewHitmarkersEnabled (BlueprintContext|GetFortSettingsContext)) "(TagName=\"SoundLibrary.HitNotify.Profile.Vehicles.Modern\")" "(TagName=\"SoundLibrary.HitNotify.Profile.Vehicles.Legacy\")")))

(event EventBeginPlay
  (Default|SelectHitmarkerTypebyTag (select (FortAudioOptions|GetNewHitmarkersEnabled (BlueprintContext|GetFortSettingsContext)) "(TagName=\"SoundLibrary.HitNotify.Profile.Vehicles.Modern\")" "(TagName=\"SoundLibrary.HitNotify.Profile.Vehicles.Legacy\")")))

(fn SelectHitmarkerTypeByTag
  (params (HitmarkerTypeTag :type Gameplay%20Tag%20Structure))
  (bind _returnvalue (Components|GetOwner))
  (bind _returnvalue_2 (WorldSubsystems|GetSoundLibrarySubsystem))
  (Utilities|IsValid _returnvalue
    (:Is%20Valid
      (LocalVariables|SetOwningActor _returnvalue)
      (SoundLibrary|RemoveLibrariesfromCollections _returnvalue_2 (LocalVariables|GetOwningActor) (GameplayTags|MakeLiteralGameplayTagContainer "(GameplayTags=((TagName=\"SoundLibrary.HitNotify.Profile.Vehicles.Legacy\"),(TagName=\"SoundLibrary.HitNotify.Profile.Vehicles.Modern\")))"))
      (SoundLibrary|AddLibrariesfromCollections _returnvalue_2 (LocalVariables|GetOwningActor) (GameplayTags|MakeGameplayTagContainerfromTag HitmarkerTypeTag)))))
