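; Sound library component for Fort pawns that sets up hit-notify sounds and picks the Modern or Legacy hitmarker audio profile.
; On BeginPlay, if its collection tags are not empty, it re-registers those collections on the owner through the Sound Library subsystem, then applies the profile from the player's audio settings.
; When the new-hitmarkers setting changes, SelectHitmarkerTypeByTag removes both profile collections and adds the chosen one.
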(event SoundLibrary|EventOnSetNewHitmarkersEnabledChanged
  (bind _returnvalue (BlueprintContext|GetFortSettingsContext))
  (Utilities|IsValid _returnvalue
    (:Is%20Valid
      (Default|SelectHitmarkerTypebyTag (select (FortAudioOptions|GetNewHitmarkersEnabled _returnvalue) "(TagName=\"SoundLibrary.HitNotify.Profile.Modern\")" "(TagName=\"SoundLibrary.HitNotify.Profile.Legacy\")")))))

(event EventBeginPlay
  (bind _returnvalue (Components|GetOwner))
  (bind _returnvalue_2 (BlueprintContext|GetFortSettingsContext))
  (bind _returnvalue_3 (WorldSubsystems|GetSoundLibrarySubsystem))
  (bind _collection_tags (Variables|Default|GetCollection%20Tags))
  ;[1 1 1] Adds HitNotify sound libraries on SL Collections
  (if (Utilities|GameplayTagContainerNotEmpty (Variables|Default|GetCollection%20Tags))
    (Utilities|IsValid _returnvalue
      (:Is%20Valid
        (SoundLibrary|RemoveLibrariesfromCollections _returnvalue_3 _returnvalue _collection_tags)
        (SoundLibrary|AddLibrariesfromCollections _returnvalue_3 _returnvalue _collection_tags)
        (Utilities|IsValid _returnvalue_2
          (:Is%20Valid
            (Default|SelectHitmarkerTypebyTag (select (FortAudioOptions|GetNewHitmarkersEnabled _returnvalue_2) "(TagName=\"SoundLibrary.HitNotify.Profile.Modern\")" "(TagName=\"SoundLibrary.HitNotify.Profile.Legacy\")"))))))))

(fn SelectHitmarkerTypeByTag
  (params (HitmarkerTypeTag :type Gameplay%20Tag%20Structure))
  (bind _returnvalue (Components|GetOwner))
  (bind _returnvalue_2 (WorldSubsystems|GetSoundLibrarySubsystem))
  (Utilities|IsValid _returnvalue
    (:Is%20Valid
      (LocalVariables|SetOwningActor _returnvalue)
      (SoundLibrary|RemoveLibrariesfromCollections _returnvalue_2 (LocalVariables|GetOwningActor) (GameplayTags|MakeLiteralGameplayTagContainer "(GameplayTags=((TagName=\"SoundLibrary.HitNotify.Profile.Legacy\"),(TagName=\"SoundLibrary.HitNotify.Profile.Modern\")))"))
      (SoundLibrary|AddLibrariesfromCollections _returnvalue_2 (LocalVariables|GetOwningActor) (GameplayTags|MakeGameplayTagContainerfromTag HitmarkerTypeTag)))))
